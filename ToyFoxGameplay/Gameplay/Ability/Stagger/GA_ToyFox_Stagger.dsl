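; Gameplay ability for ToyFox that plays a directional stagger montage on activation.
; A helper picks Back, Left, Right or Forward by the azimuth of the granting effect's instigator relative to the avatar pawn, falling back to Back.
; EndAbility follows the montage task directly, with all task callbacks left empty.
; On end, the ability removes its granting gameplay effect before calling the parent.
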(event Ability|EventActivateAbility
  ;[0.007489 0.15 0 0.5] Activate: Play the stagger montage.
  (|CallParentFunction)
  (Ability|Tasks|PlayMontageAndWaitNotify "StaggerMontage" (CallFunction|GetStaggerMontagetoPlay)
    (:OnCompleted
      _)
    (:OnBlendOut
      _)
    (:OnInterrupted
      _)
    (:OnCancelled
      _))
  (Ability|EndAbility))

(event Ability|EventOnEndAbility
  (params (bWasCancelled :type Boolean))
  ;[0.1042 0.02416 0 0.5] OnEndAbility: Remove the granting gameplay effect.
  (Utilities|FlowControl|Sequence
    (:then_0
      (Ability|RemoveGrantedbyEffect))
    (:then_1
      (|CallParentFunction bWasCancelled))))

(fn GetStaggerMontageToPlay
  (Utilities|FlowControl|Sequence
    (:then_0
      (bind _returnvalue (Ability|EffectContext|GetInstigatorActor (Ability|GetGrantedbyEffectContext)))
      (Utilities|IsValid _returnvalue
        (:Is%20Valid
          (bind _outfortplayerpawn (CallFunction|GetAvatarasValidFortPlayerPawn)
            (:ValidFortPlayerPawn
              (bind _azimuth (Math|Vector|GetAzimuthandElevation (Math|Vector|GetUnitDirection(Vector) (Transformation|GetActorLocation _outfortplayerpawn) (Transformation|GetActorLocation _returnvalue)) (Transformation|GetActorTransform _outfortplayerpawn)))
              (if (and (> _azimuth -45.0) (< _azimuth 45.0))
                (return (Variables|Default|GetStaggerMontage_Back))
                (elif (and (> _azimuth 45.0) (< _azimuth 135.0))
                  (return (Variables|Default|GetStaggerMontage_Left))
                  (elif (and (> _azimuth -135.0) (< _azimuth -45.0))
                    (return (Variables|Default|GetStaggerMontage_Right))
                    (else
                      (return (Variables|Default|GetStaggerMontage_Forward)))))))))))
    (:then_1
      (return (Variables|Default|GetStaggerMontage_Back)))))
